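# Sound cue for the Stink Flopper aura loop, layering fish jump and gas sounds.
# A mixer blends a randomly picked jump one-shot, shaped by a short volume fade-in envelope, with a loud gas loop and a quieter, pitch-modulated high gas loop.
# A root modulator varies the whole mix.
# The cue uses consumable sound class and attenuation settings, with concurrency capped at two instances.
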SoundCue /FNGameplayCues/Athena/Sounds/Items/Flopper/StinkFlopper/Stink_Flopper_Aura_Loop_Cue
root=Modulator_0
nodes [
 {
  id=Enveloper_0
  children [Random_0]
  properties {
   pitchCurve {editorCurveData {keys [{time=0}]}}
   volumeCurve {editorCurveData {keys [<at 0.0036 0.0389>,<at 0.0641 0.2406>,<at 0.1085 0.5858>,<at 0.1475 0.967>]}}
  }
 }
 {id=Mixer_0, children [Enveloper_0, WavePlayer_15, Modulator_5], properties {inputVolume [0.3,1.2,0.175]}}
 {id=Modulator_0, children [Mixer_0]}
 {id=Modulator_5, children [WavePlayer_2], properties {pitchMax=1.3, pitchMin=0.98}}
 {id=Random_0, children [WavePlayer_3, WavePlayer_4, WavePlayer_12, WavePlayer_13], properties {weights [1,1,1,1]}}
 {id=WavePlayer_12, properties {soundWaveAssetPtr=&/Stink_Fish_Jump_03}}
 {id=WavePlayer_13, properties {soundWaveAssetPtr=&/Stink_Fish_Jump_04}}
 {id=WavePlayer_15, properties {bLooping=true, soundWaveAssetPtr=/Game/Athena/Sounds/Items/Flopper/StinkFlopper/Stink_Fish_Gas_Loop}}
 {id=WavePlayer_2, properties {bLooping=true, soundWaveAssetPtr=/Game/Athena/Sounds/Items/Flopper/StinkFlopper/Stink_Fish_High_Gas_Loop}}
 {id=WavePlayer_3, properties {soundWaveAssetPtr=&/Stink_Fish_Jump_01}}
 {id=WavePlayer_4, properties {soundWaveAssetPtr=&/Stink_Fish_Jump_02}}
]
properties {
 attenuationSettings=&Fort_Consumable_Use_Athena_Att
 bOverrideConcurrency=true
 maxDistance=2250
 soundClassObject=&Fort_Consumables
 volumeMultiplier=0.5
 concurrencyOverrides {maxCount=2}
}
