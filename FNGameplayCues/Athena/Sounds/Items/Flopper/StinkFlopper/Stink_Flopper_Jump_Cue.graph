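# Sound cue for the Stink Flopper item's jump sound in Athena.
# A random node with equal weights picks one of four fish jump waves and feeds a modulator at the root.
# It plays quietly through the explosives sound class with close explosion attenuation.
# A concurrency override caps simultaneous instances at two and scales down volume between them.
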SoundCue /FNGameplayCues/Athena/Sounds/Items/Flopper/StinkFlopper/Stink_Flopper_Jump_Cue
root=Modulator_0
nodes [
 {id=Modulator_0, children [Random_1]}
 {id=Random_1, children [WavePlayer_0,WavePlayer_1,WavePlayer_2,WavePlayer_3], properties {weights [1,1,1,1]}}
 {id=WavePlayer_0, properties {soundWaveAssetPtr=&/Stink_Fish_Jump_01}}
 {id=WavePlayer_1, properties {soundWaveAssetPtr=&/Stink_Fish_Jump_02}}
 {id=WavePlayer_2, properties {soundWaveAssetPtr=&/Stink_Fish_Jump_03}}
 {id=WavePlayer_3, properties {soundWaveAssetPtr=&/Stink_Fish_Jump_04}}
]
properties {
 attenuationSettings=/Game/Sounds/Attenuation/Fort_Explosion_Close_StW_Att
 bOverrideConcurrency=true
 maxDistance=3500
 soundClassObject=/Game/Sounds/FortSoundClassesAndModes/Fort_Explosives
 volumeMultiplier=0.2
 concurrencyOverrides {maxCount=2, volumeScale=0.7}
}
